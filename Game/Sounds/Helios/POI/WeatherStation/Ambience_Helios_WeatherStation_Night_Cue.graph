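# Looping night ambience cue for the Helios weather station POI, routed to the outdoor ambient sound class with ambient attenuation.
# A quality-level node selects the quad exterior loop for its first two levels and the stereo loop for the other two.
# Each loop passes through nested, endlessly looping envelopers, one modulating pitch and the other dipping volume.
# Their differing cycle lengths make the combined variation drift, so the loop sounds less repetitive.
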SoundCue /Game/Sounds/Helios/POI/WeatherStation/Ambience_Helios_WeatherStation_Night_Cue
root=QualityLevel_1
nodes [
 {
  id=Enveloper_0
  children [WavePlayer_1]
  properties {
   bLoop=true
   bLoopIndefinitely=true
   loopEnd=40
   pitchCurve {editorCurveData {keys [
      {arriveTangent=0, interpMode=RCIM_Cubic, leaveTangent=0, tangentMode=RCTM_User, time=0}
      {arriveTangent=0, interpMode=RCIM_Cubic, leaveTangent=0, tangentMode=RCTM_User, time=20, value=0.92}
      {arriveTangent=0.004, interpMode=RCIM_Cubic, leaveTangent=0.004, time=40}
     ]}}
   volumeCurve {editorCurveData {keys [<at 0 1>]}}
  }
 }
 {
  id=Enveloper_4
  children [WavePlayer_0]
  properties {
   bLoop=true
   bLoopIndefinitely=true
   loopEnd=40
   pitchCurve {editorCurveData {keys [
      {arriveTangent=0, interpMode=RCIM_Cubic, leaveTangent=0, tangentMode=RCTM_User, time=0}
      {arriveTangent=0, interpMode=RCIM_Cubic, leaveTangent=0, tangentMode=RCTM_User, time=20, value=0.92}
      {arriveTangent=0.004, interpMode=RCIM_Cubic, leaveTangent=0.004, time=40}
     ]}}
   volumeCurve {editorCurveData {keys [<at 0 1>]}}
  }
 }
 {
  id=Enveloper_5
  children [Enveloper_4]
  properties {
   bLoop=true
   bLoopIndefinitely=true
   loopEnd=30
   pitchCurve {editorCurveData {keys [{time=0}]}}
   volumeCurve {editorCurveData {keys [
      <at 0 1 RCIM_Cubic>
      {arriveTangent=0, interpMode=RCIM_Cubic, leaveTangent=0, tangentMode=RCTM_User, time=15, value=0.8}
      <at 30 1 RCIM_Cubic>
     ]}}
  }
 }
 {
  id=Enveloper_6
  children [Enveloper_0]
  properties {
   bLoop=true
   bLoopIndefinitely=true
   loopEnd=30
   pitchCurve {editorCurveData {keys [{time=0}]}}
   volumeCurve {editorCurveData {keys [
      <at 0 1 RCIM_Cubic>
      {arriveTangent=0, interpMode=RCIM_Cubic, leaveTangent=0, tangentMode=RCTM_User, time=15, value=0.8}
      <at 30 1 RCIM_Cubic>
     ]}}
  }
 }
 {id=QualityLevel_1, children [Enveloper_6,Enveloper_6,Enveloper_5,Enveloper_5]}
 {
  id=WavePlayer_0
  properties {bLooping=true, soundWaveAssetPtr=&/Ambience_WeatherStation_Night_Loop_Ext_Stereo}
 }
 {
  id=WavePlayer_1
  properties {bLooping=true, soundWaveAssetPtr=&/Ambience_WeatherStation_Night_Loop_Ext_Quad}
 }
]
properties {
 attenuationSettings=&Fort_Ambient_Athena_Att
 maxDistance=2097152
 soundClassObject=&Fort_Ambient_Outdoors_Athena
 volumeMultiplier=0.251
}
